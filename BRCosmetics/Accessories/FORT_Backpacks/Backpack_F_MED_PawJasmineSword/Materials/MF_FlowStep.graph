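# Material function for the PawJasmineSword back bling that outputs tinted flow bands positioned by a Flow Step input rather than time.
# The tiled gradient and step are summed, wrapped with Frac and scaled by FlowLength, then remapped through named reroutes for bias, smooth or linear falloff and an exponent curve.
# A fire noise texture feeds optional contrast-adjusted breakup, and a mask input, color and intensity shape the output.
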MaterialFunction /BRCosmetics/Accessories/FORT_Backpacks/Backpack_F_MED_PawJasmineSword/Materials/MF_FlowStep
graph {nodes [
  n1 = VectorParameter() {group=Flow, parameterName=FlowColor, defaultValue=<rgba 1 1 1 1>}
  n2 = Add(A=Subtract(A=NamedRerouteUsage(), B=Constant() {r=1}), B=Multiply(A=NamedRerouteUsage(), B=ScalarParameter() {defaultValue=2, group=Flow}))
  n3 = ScalarParameter() {group=Flow, parameterName=FlowNoiseContrast}
  n4 = Multiply(
   A=LinearInterpolate(
    A=NamedRerouteUsage()
    B=MaterialFunctionCall(In=@n2, Contrast=@n3) {materialFunction=&CheapContrast}.Result
    Alpha=ScalarParameter() {group=Flow, parameterName=FlowNoiseStrength}
   )
   B=NamedRerouteUsage()
  )
  FunctionOutput(None=Multiply(A=Multiply(A=@n1.RGB, B=@n4), B=ScalarParameter() {defaultValue=1, group=Flow, parameterName=FlowIntensity}))
  MaterialFunctionCall!(X range Y range, Gradient Height, ShowGrid?, Thickness, Color, "f(x)"=NamedRerouteUsage()) {materialFunction=/Engine/Functions/Engine_MaterialFunctions01/Debug/PlotFunctionOnGraph}
  n5 = NamedRerouteUsage()
  NamedRerouteDeclaration(None=Power(Base=OneMinus(None=@n5), Exp=ScalarParameter() {defaultValue=2, group=Flow, parameterName=FlowCurveExponent})) {name=Curve, nodeColor=<rgba 0 0.7656 1 1>}
  n6 = NamedRerouteUsage()
  n7 = NamedRerouteUsage()
  n8 = SmoothStep(Min=@n6, Value=@n7)
  NamedRerouteDeclaration(None=If(A=NamedRerouteUsage(), B=NamedRerouteUsage(), "A > B"=@n8, "A == B"=@n8, "A < B"=SmoothStep(Min=@n6, Value=@n7) {constMax=-1})) {name=Smooth, nodeColor=<rgba 0.3125 0 1 1>}
  n9 = NamedRerouteUsage()
  n10 = NamedRerouteUsage()
  n11 = InverseLinearInterpolate(A=@n9, Value=@n10) {constA=0.5}
  NamedRerouteDeclaration(None=If(
    A=NamedRerouteUsage()
    B=NamedRerouteUsage()
    "A > B"=@n11
    "A == B"=@n11
    "A < B"=InverseLinearInterpolate(A=@n9, Value=@n10) {constA=0.5, constB=-1}
   )) {name=Linear, nodeColor=<rgba 0.1719 0 1 1>}
  n12 = TextureCoordinate()
  n13 = ScalarParameter() {defaultValue=1, group=Flow, parameterName=FlowTiling}
  n14 = Multiply(
   A=FunctionInput(Preview=ComponentMask(None=@n12) {r=true}) {bUsePreviewValueAsDefault=true, inputName=FlowGradient, inputType=FunctionInput_Scalar}
   B=@n13
  ) {constB=10}
  n15 = FunctionInput() {bUsePreviewValueAsDefault=true, inputName=Flow Step, inputType=FunctionInput_Scalar, sortPriority=3, previewValue=<1 0 0 1>}
  n16 = Multiply(A=@n13, B=@n15)
  n17 = Add(A=@n14, B=@n16)
  n18 = Saturate(None=Divide(A=Frac(None=@n17), B=ScalarParameter() {defaultValue=0.2, group=Flow, parameterName=FlowLength}) {constB=0.2})
  NamedRerouteDeclaration(None=ConstantBiasScale(None=OneMinus(None=@n18)) {bias=-0.5, scale=2}) {name=GradientRemap, nodeColor=<rgba 0 0.2969 1 1>}
  NamedRerouteDeclaration(None=ScalarParameter() {group=Flow, parameterName=FlowGradientBias}) {name=Bias, nodeColor=<rgba 0.8281 0 1 1>}
  n19 = ComponentMask(None=TextureCoordinate()) {g=true}
  n20 = Constant() {r=0.5}
  NamedRerouteDeclaration(None=FunctionInput(Preview=SphereMask(A=@n19, B=@n20) {attenuationRadius=0.1}) {bUsePreviewValueAsDefault=true, inputName=FlowMask, inputType=FunctionInput_Scalar, sortPriority=1}) {name=FlowMask, nodeColor=<rgba 0 1 0.1562 1>}
  n21 = TextureCoordinate()
  n22 = ScalarParameter() {defaultValue=1, group=Flow, parameterName=NoiseTiling}
  NamedRerouteDeclaration(None=TextureSampleParameter2D(UVs=Multiply(A=@n21, B=@n22) {constB=5}) {group=Flow, parameterName=FlowNoise, texture=/Game/Global/Textures/Default/Noise/T_FireNoise_01}.R) {name=Noise, nodeColor=<rgba 0 1 0.7422 1>}
  Subtract(A=@n16, B=@n14)
 ]}
