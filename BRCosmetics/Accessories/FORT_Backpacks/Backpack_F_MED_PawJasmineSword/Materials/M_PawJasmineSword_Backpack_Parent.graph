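# Parent material for the PawJasmineSword backpack cosmetic, inserting custom emissive effects between the 2023 character parent PreFX and PostFX functions.
# A parameter-driven time step sequentially blends four packed mask channels into a smoothstepped, tinted arrow, added to a flow-step effect.
# A reactivity parameter gates that scrolling glow, which Max-combines with base emissive shaped by swoosh intensity and the Tech Art mask alpha.
# Base color darkens where effect masks apply.
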Material /BRCosmetics/Accessories/FORT_Backpacks/Backpack_F_MED_PawJasmineSword/Materials/M_PawJasmineSword_Backpack_Parent
graph {
 properties {
  bHasPixelAnimation=true
  bUseMaterialAttributes=true
  bUsedWithMeshDeformer=true
  bUsedWithMorphTargets=true
  bUsedWithSkeletalMesh=true
  pixelDepthOffsetMode=PDOM_Legacy
  previewMesh=/BRCosmetics/Accessories/FORT_Backpacks/Backpack_F_MED_PawJasmineSword/Meshes/F_MED_PawJasmineSword_Pack
  shadingModel=MSM_FromMaterialExpression
  usedShadingModels=Default Lit | Subsurface | Clear Coat | Subsurface Profile | Cloth
 }
 nodes [
  n1 = MaterialFunctionCall() {materialFunction=&MF_CharacterParent_2023_PreFX}
  n2 = GetMaterialAttributes(None=@n1.Result) {attributeGetTypes [#1, #2]}
  n3 = NamedRerouteDeclaration(None=@n2.Emissive Color) {name=Emissive Color, nodeColor=<rgba 0.1016 0 1 1>}
  n4 = NamedRerouteUsage()
  n5 = NamedRerouteDeclaration(None=Multiply(A=@n4, B=ScalarParameter() {defaultValue=1, parameterName=Swoosh Intensity})) {name=Swooshie, nodeColor=<rgba 0 0.0625 1 1>}
  n6 = OneMinus(None=Step(Y=@n1.Tech Art Mask Alpha) {constX=0})
  n7 = VectorParameter() {parameterName=Base Emissive COlor, defaultValue=<rgba 0.6172 0.0482 0.0052 1>}
  n8 = Max(A=Subtract(A=@n3, B=@n4), B=LinearInterpolate(A=Multiply(A=@n3, B=@n5), B=Multiply(A=@n6, B=@n7.RGB), Alpha=@n6))
  n9 = ScalarParameter() {defaultValue=0.1, parameterName=Line - Step}
  n10 = ScalarParameter() {defaultValue=0.75, parameterName=Line - Step Smoothness}
  n11 = TextureSample() {
   samplerType=SAMPLERTYPE_Masks
   texture=/BRCosmetics/Accessories/FORT_Backpacks/Backpack_F_MED_PawJasmineSword/Textures/T_PawJasmineSword_Backpack_FX2
  }
  n12 = ScalarParameter() {parameterName=Scrolling - Time Step}
  n13 = ScalarParameter() {defaultValue=1, parameterName=Scrolling - Speed}
  n14 = Multiply(A=Fmod(A=Multiply(A=@n12, B=@n13), B=Constant() {r=1})) {constB=5}
  n15 = LinearInterpolate(B=@n11.R, Alpha=Saturate(None=@n14))
  n16 = Subtract(A=@n14)
  n17 = Saturate(None=@n16)
  n18 = Subtract(A=@n16)
  n19 = Subtract(A=@n18)
  n20 = LinearInterpolate(
   A=LinearInterpolate(A=LinearInterpolate(A=@n15, B=@n11.G, Alpha=@n17), B=@n11.B, Alpha=Saturate(None=@n18))
   B=@n11.A
   Alpha=Saturate(None=@n19)
  )
  n21 = Saturate(None=Subtract(A=@n19))
  n22 = NamedRerouteDeclaration(None=SmoothStep(Min=@n9, Max=Add(A=@n9, B=@n10), Value=LinearInterpolate(A=@n20, Alpha=@n21) {constB=0})) {name=Arrow, nodeColor=<rgba 0 0.8594 1 1>}
  n23 = VectorParameter() {parameterName=Scrolling - Color, defaultValue=<rgba 0.7991 0.3162 0 1>}
  n24 = MaterialFunctionCall!(FlowGradient=@n1.Tech Art Mask Alpha, FlowMask=@n1.Tech Art Mask Alpha, Flow Step=Add(A=@n12) {constB=0.25}) {materialFunction=&/MF_FlowStep}
  n25 = Power(Base=Desaturation(None=NamedRerouteUsage()), Exp=ScalarParameter() {defaultValue=2, parameterName=Exp})
  n26 = Multiply(
   A=Add(A=Multiply(A=@n22, B=@n23.RGB), B=Multiply(A=@n24.Result, B=@n25))
   B=LinearInterpolate(B=ScalarParameter() {defaultValue=2, parameterName=Scrolling - Intensity}, Alpha=ScalarParameter() {parameterName=Enable Reactivity?})
  )
  n27 = Max(A=NamedRerouteUsage(), B=NamedRerouteUsage())
  MaterialFunctionCall(
   Material=SetMaterialAttributes!(
    MaterialAttributes=@n2.MaterialAttributes
    Emissive Color=Max(A=@n8, B=@n26)
    Base Color=LinearInterpolate(A=@n2.Base Color, Alpha=@n27) {constB=0}
   ) {attributeSetTypes [#1, #2]}
   M=@n1.M
   Pre Skinned Local Position=@n1.M
   Distance Blend=@n1.M
   Fresnel=@n1.M
  ) {materialFunction=&MF_CharacterParent_2023_PostFX}
  n28 = TextureSample() {
   samplerType=SAMPLERTYPE_Grayscale
   texture=/BRCosmetics/Accessories/FORT_Backpacks/Backpack_F_MED_PawJasmineSword/Textures/T_PawJasmineSword_Backpack_Swoosh
  }
  NamedRerouteDeclaration(None=@n28.RGB) {name=Swoosh, nodeColor=<rgba 0.125 1 0 1>}
  NamedRerouteUsage()
 ]
}
